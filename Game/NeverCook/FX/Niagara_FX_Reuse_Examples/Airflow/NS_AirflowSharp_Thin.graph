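# Niagara system for a thin, sharp airflow effect, kept among reusable FX examples and built from one mesh-rendered emitter.
# An instantaneous burst on a one-second emitter duration spawns particles squashed to half scale in Z, with material random enabled.
# Particle update writes four dynamic material parameters, leaving the look and any motion to the material, as no movement modules appear.
# User parameters declare total scale, front-end flags and a skeletal mesh, with no visible bindings.
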NiagaraSystem /Game/NeverCook/FX/Niagara_FX_Reuse_Examples/Airflow/NS_AirflowSharp_Thin
dependencies {usedDataInterfaces [], usedDynamicInputs [], usedModules [&SystemState]}
emitters {AirflowSharp_Thin {
  emitterUpdateScript [
   &EmitterState(Life Cycle Mode=::NewEnumerator1, Loop Duration=1, UseLoopDelay=false)
   &SpawnBurst_Instantaneous(Use Spawn Probability=false, UseLoopCountLimit=false)
  ]
  particleSpawnScript [&InitializeParticle(Mesh Scale=<1 1 0.5>, Mesh Scale Mode=::NewEnumerator4, UsePositionOffset=false, Write Material Random=true)]
  particleUpdateScript [
   &ParticleState()
   &DynamicMaterialParameters(
    Index 0 Param 1=&RandomRangeFloat(Minimum=0, Maximum=0)
    Index 0 Param 2=0.5
    Index 0 Param 3=2
    Module.Param0WriteEnabled=true
    Module.Param1WriteEnabled=true
    Module.Param2WriteEnabled=true
    Module.Param3WriteEnabled=true
   )
  ]
  rendererClasses [&NiagaraMeshRendererProperties]
  renderers [{rendererClass=&NiagaraMeshRendererProperties, rendererIndex=0}]
 }}
summary {
 systemName=NS_AirflowSharp_Thin
 emitters [{emitterName=AirflowSharp_Thin, rendererClasses [&NiagaraMeshRendererProperties]}]
 userVariables [
  User.bIsFrontEnd {type=NiagaraBool}
  User.bIsFrontEndPreview {type=NiagaraBool}
  User.Skeletal Mesh {
   type=NiagaraDataInterfaceSkeletalMesh
   defaultValue {
    struct=NiagaraExt_VariableValue_DataInterface
    value {dataInterface=&self:NiagaraDataInterfaceSkeletalMesh_1}
   }
  }
  User.Total Scale {type=NiagaraFloat, defaultValue=1, description=""}
 ]
}
