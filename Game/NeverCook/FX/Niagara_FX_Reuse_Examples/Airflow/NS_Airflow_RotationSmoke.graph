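# Niagara system for a swirling smoke airflow effect in a reusable FX examples folder, built from a single mesh-rendered Ring emitter.
# The emitter continuously spawns short-lived, vertically stretched meshes with random scale and orientation inside a narrow cylinder.
# Over each particle's life, curves grow the mesh, slow its spin around Z to a stop, and fade alpha in and out.
# A color curve tints the grey base purple, briefly brightening before dimming.
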NiagaraSystem /Game/NeverCook/FX/Niagara_FX_Reuse_Examples/Airflow/NS_Airflow_RotationSmoke
dependencies {
 usedDataInterfaces [&NiagaraDataInterfaceCurve, &NiagaraDataInterfaceColorCurve]
 usedDynamicInputs []
 usedModules [&SystemState]
}
emitters {Ring {
  emitterUpdateScript [
   &EmitterState(Life Cycle Mode=::NewEnumerator1, Loop Duration=1, UseLoopDelay=false)
   &SpawnRate(SpawnRate=10, Use Spawn Probability=false)
  ]
  particleSpawnScript [
   &InitializeParticle(
    Color Maximum=<rgba 0.3 0.3 0.3 1>
    Color Minimum=<rgba 0.2 0.2 0.2 1>
    Color Mode=::NewEnumerator2
    Lifetime Max=1
    Lifetime Min=0.5
    Lifetime Mode=::NewEnumerator1
    Mesh Scale=&Multiply_VectorByFloat(
     Vector=&RandomRangeVector(Minimum=<1 1 1.4>, Maximum=<1 1 1.8>)
     Float=&RandomRangeFloat(Minimum=0.45, Maximum=0.55)
    )
    Mesh Scale Mode=::NewEnumerator4
    UsePositionOffset=false
    Write Material Random=true
   )
   &InitialMeshOrientation(
    Mesh Orientation Mode=::NewEnumerator2
    Rotation=&RandomRangeVector(Minimum=<0.02 0.02 -1>, Maximum=<0.03 0.03 1>)
    Use Rotation Vector=true
   )
   &ShapeLocation(
    Cylinder Height=50
    Cylinder Radius=1
    MaskSpawnGroup=false
    Offset Mode=::NewEnumerator0
    Override Seed=false
    Shape Primitive=::NewEnumerator1
    Surface Only=false
    UseOwnerScale=true
   )
  ]
  particleUpdateScript [
   &ParticleState()
   &ScaleMeshSize(Scale Factor=&Multiply_VectorByFloat(
     Vector=<1 1 1>
     Float=&FloatFromCurve(
      FloatCurve=DataInterface() {propertyValues {Curve {keys [
          {arriveTangent=0.5864, interpMode=RCIM_Cubic, leaveTangent=0.5864, tangentMode=RCTM_User, time=0, value=0.5}
          {arriveTangent=0.4564, interpMode=RCIM_Cubic, leaveTangent=0.4564, tangentMode=RCTM_User, time=0.5954, value=0.6852}
          {arriveTangent=0.5662, interpMode=RCIM_Cubic, leaveTangent=0.5662, tangentMode=RCTM_User, time=1, value=1.1}
         ]}}}
      Scale Curve=4
     )
    ))
   &UpdateMeshOrientation(
    Rotation Rate=&Multiply_Float(
     A=-0.8
     B=&FloatFromCurve(FloatCurve=DataInterface() {propertyValues {Curve {keys [<at 0 1>,<at 0.4 1>,<at 1 0>]}}})
    )
    Rotation Vector=<0 0 1>
   )
   &ScaleColor(
    Scale Alpha=&FloatFromCurve(
     FloatCurve=DataInterface() {propertyValues {Curve {keys [
         {arriveTangent=4.9076, interpMode=RCIM_Cubic, leaveTangent=4.9076, tangentMode=RCTM_User, time=0, value=0}
         {arriveTangent=0.0332, interpMode=RCIM_Cubic, leaveTangent=0.0332, tangentMode=RCTM_User, time=0.1, value=1}
         {arriveTangent=-0.0163, interpMode=RCIM_Cubic, leaveTangent=-0.025, tangentMode=RCTM_Break, time=0.4, value=1}
         {arriveTangent=-1.5, interpMode=RCIM_Cubic, leaveTangent=-1.5, time=0.6, value=0.8}
         {arriveTangent=-2, interpMode=RCIM_Cubic, leaveTangent=-2, time=0.9, value=0.25}
         {arriveTangent=-3.0089, tangentMode=RCTM_User, time=1, value=0}
        ]}}}
     Scale Curve=2
    )
    Scale RGB=&MakeVectorFromLinearColorRGB(Color=&ColorFromCurve(ColorCurve=DataInterface() {propertyValues {
        AlphaCurve {keys []}
        BlueCurve {keys [<at 0.0254 0.5>, {arriveTangent=-0.3927, interpMode=RCIM_Cubic, leaveTangent=-0.3927, time=0.2471, value=1.5}, <at 0.6621 0.25 RCIM_Cubic>]}
        GreenCurve {keys [
          <at 0.0254 0.025>
          {arriveTangent=-0.0196, interpMode=RCIM_Cubic, leaveTangent=-0.0196, time=0.2471, value=0.165}
          <at 0.6621 0.0125 RCIM_Cubic>
         ]}
        RedCurve {keys [
          <at 0.0254 0.2577>
          {arriveTangent=-0.2024, interpMode=RCIM_Cubic, leaveTangent=-0.2024, time=0.2471, value=0.8325}
          <at 0.6621 0.1289 RCIM_Cubic>
         ]}
       }}))
    ScaleA=true
    ScaleRGB=true
   )
   &SolveForcesAndVelocity(Clamp Velocity=false, Limit Acceleration=false)
  ]
  rendererClasses [&NiagaraMeshRendererProperties]
  renderers [{rendererClass=&NiagaraMeshRendererProperties, rendererIndex=0}]
 }}
summary {
 systemName=NS_Airflow_RotationSmoke
 emitters [{emitterName=Ring, rendererClasses [&NiagaraMeshRendererProperties]}]
 userVariables [
  User.bIsFrontEnd {type=NiagaraBool}
  User.bIsFrontEndPreview {type=NiagaraBool}
 ]
}
