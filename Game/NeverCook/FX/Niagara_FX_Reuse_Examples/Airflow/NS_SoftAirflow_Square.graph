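# Niagara reuse-example effect that, judging by its name, shows a soft, square-shaped airflow as ribbons.
# A single emitter spawns particles at a steady rate, runs PartitionParticles, then initializes them with a grey color, set ribbon width and material random values.
# Particles get a constant +X velocity and feed a ribbon renderer, with dynamic material parameters set at spawn.
# User parameters expose front-end flags and a total scale, with no visible use in the emitter.
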NiagaraSystem /Game/NeverCook/FX/Niagara_FX_Reuse_Examples/Airflow/NS_SoftAirflow_Square
dependencies {usedDataInterfaces [], usedDynamicInputs [], usedModules [&SystemState]}
emitters {Example {
  emitterUpdateScript [
   &EmitterState(Life Cycle Mode=::NewEnumerator1, Loop Behavior=::NewEnumerator1, Loop Duration Mode=::NewEnumerator1, UseLoopDelay=false)
   &SpawnRate(SpawnRate=100, Use Spawn Probability=false)
  ]
  particleSpawnScript [
   &PartitionParticles()
   &InitializeParticle(
    Color=<rgba 0.4427 0.4427 0.4427 1.5>
    Mesh Scale=<0.55 1 1.5>
    Mesh Scale Mode=::NewEnumerator4
    Position Offset=<6 -6 0>
    Ribbon Width=100
    Ribbon Width Mode=::NewEnumerator1
    UsePositionOffset=false
    Write Material Random=true
   )
   &DynamicMaterialParameters(
    Index 0 Param 3=1
    Index 0 Param 4=0.5
    Module.Index0Param0Write=true
    Module.Index0Param1Write=true
    Module.Index0Param2Write=true
    Module.Index0Param3Write=true
   )
   &AddVelocity(Velocity=<300 0 0>)
  ]
  particleUpdateScript [
   &SolveForcesAndVelocity(Clamp Velocity=false, Limit Acceleration=false)
   &ParticleState()
  ]
  rendererClasses [&NiagaraRibbonRendererProperties]
  renderers [{rendererClass=&NiagaraRibbonRendererProperties}]
 }}
summary {
 systemName=NS_SoftAirflow_Square
 emitters [{emitterName=Example, rendererClasses [&NiagaraRibbonRendererProperties]}]
 userVariables [
  User.bIsFrontEnd {type=NiagaraBool}
  User.bIsFrontEndPreview {type=NiagaraBool}
  User.Total Scale {type=NiagaraFloat, defaultValue=1, description=""}
 ]
}
