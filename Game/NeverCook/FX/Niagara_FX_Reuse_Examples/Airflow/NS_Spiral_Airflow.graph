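# NS_Spiral_Airflow is a Niagara effect, filed as a reuse example, that draws a spiral airflow from mesh particles.
# Its single AirflowRemain emitter combines continuous spawning with a small burst, placing vertically stretched, raised meshes with random yaw.
# Meshes spin around Z and widen horizontally over life while curve-driven and random values feed dynamic material parameters.
# Color fades in and out by curve, with tint and brightness following time of day through FNTimeOfDay inputs.
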NiagaraSystem /Game/NeverCook/FX/Niagara_FX_Reuse_Examples/Airflow/NS_Spiral_Airflow
dependencies {
 usedDataInterfaces [&NiagaraDataInterfaceVectorCurve, &NiagaraDataInterfaceCurve]
 usedDynamicInputs []
 usedModules [&SystemState]
}
emitters {AirflowRemain {
  emitterUpdateScript [
   &EmitterState(Life Cycle Mode=::NewEnumerator1, Loop Delay=0.5, Loop Duration=0.8, UseLoopDelay=false)
   &SpawnRate(SpawnRate=10, Use Spawn Probability=false)
   &SpawnBurst_Instantaneous(Spawn Count=2, Use Spawn Probability=false, UseLoopCountLimit=false)
  ]
  particleSpawnScript [
   &InitializeParticle(
    Color=<rgba 0.8542 0.9725 1 1>
    Mesh Scale=<1 1 2>
    Mesh Scale Mode=::NewEnumerator4
    Position Offset=<0 0 30>
    UsePositionOffset=true
    Write Material Random=true
   )
   &InitialMeshOrientation(
    Mesh Orientation Mode=::NewEnumerator2
    Rotation=&RandomRangeVector(Minimum=<0 0 0>, Maximum=<0 0 1>)
    Use Rotation Vector=true
   )
  ]
  particleUpdateScript [
   &ParticleState()
   &UpdateMeshOrientation(Rotation Rate=0.1, Rotation Vector=<0 0 1>)
   &ScaleMeshSize(Scale Factor=&Multiply_VectorByFloat(
     Vector=&VectorFromCurve(
      VectorCurve=DataInterface() {propertyValues {XCurve {keys [<at 0 0.6>,<at 1 1.4>]}, YCurve {keys [<at 0 0.6>,<at 1 1.4>]}, ZCurve {keys [<at 0 1>,<at 1 1>]}}}
      Scale Curve=&Multiply_VectorByFloat(Vector=<1 1 1>, Float=1)
     )
     Float=1
    ))
   &DynamicMaterialParameters(
    Index 0 Param 1=&FloatFromCurve(FloatCurve=DataInterface() {propertyValues {Curve {keys [<at 0 0.1>,<at 1 0.5>]}}})
    Index 0 Param 2=&RandomRangeFloat(Minimum=40, Maximum=60)
    Module.Index0Param0Write=true
    Module.Index0Param1Write=true
    Module.Index0Param2Write=true
    Module.Index0Param3Write=true
   )
   &ScaleColor(
    Scale RGB=&Multiply_VectorByFloat(
     Vector=&MakeVectorFromLinearColorRGB(Color=&FNTimeOfDayColor(
       Night Color=<rgba 0.474 0.6858 1 1>
       Morning Color=<rgba 0.75 0.8507 1 1>
       Day Color=<rgba 0.75 0.8507 1 1>
       Evening Color=<rgba 0.75 0.8507 1 1>
      ))
     Float=&FloatFromCurve(
      FloatCurve=DataInterface() {propertyValues {Curve {keys [
          <at 0 0 RCIM_Cubic>
          <at 0.2 1 RCIM_Cubic>
          {arriveTangent=-0.6427, interpMode=RCIM_Cubic, leaveTangent=-0.6427, tangentMode=RCTM_User, time=1, value=0}
         ]}}}
      Scale Curve=&Multiply_Float(A=1, B=&FNTimeOfDayScalar(Night Scale=0.8, Morning Scale=0.8, Evening Scale=0.8))
     )
    )
    ScaleA=true
    ScaleRGB=true
   )
  ]
  rendererClasses [&NiagaraMeshRendererProperties]
  renderers [{rendererClass=&NiagaraMeshRendererProperties, rendererIndex=0}]
 }}
summary {
 systemName=NS_Spiral_Airflow
 emitters [{emitterName=AirflowRemain, rendererClasses [&NiagaraMeshRendererProperties]}]
 userVariables []
}
